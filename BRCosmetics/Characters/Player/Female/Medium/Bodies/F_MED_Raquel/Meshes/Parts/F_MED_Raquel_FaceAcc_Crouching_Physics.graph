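# Physics asset for the crouching variant of Raquel's face accessory mesh, setting up dynamic hair over an upper-body proxy.
# Five capsule chains (left and right front; center, left and right back) hang from base bones constrained to the head, with limited swing and tight twist.
# Capsules on head, neck, upper spine, clavicles and upper arms form the body shapes, joined by free angular constraints.
# Two oversized neck_02 capsules and a large right-arm box add bulk.
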PhysicsAsset /BRCosmetics/Characters/Player/Female/Medium/Bodies/F_MED_Raquel/Meshes/Parts/F_MED_Raquel_FaceAcc_Crouching_Physics
bodies {
 base_C_hair_back [{shapeType=Capsule, length=1, radius=1, shapeName=base_C_hair_back_capsule, rotation=<rot -90 0 0>}]
 base_L_hair_back [{shapeType=Capsule, length=1, radius=1, shapeName=base_L_hair_back_capsule, rotation=<rot -90 0 0>}]
 base_L_hair_front [{shapeType=Capsule, length=1, radius=1, shapeName=base_L_hair_front_capsule}]
 base_R_hair_back [{shapeType=Capsule, length=1, radius=1, shapeName=base_R_hair_back_capsule, rotation=<rot -90 0 0>}]
 base_R_hair_front [{shapeType=Capsule, length=1, radius=1, shapeName=base_R_hair_front_capsule}]
 clavicle_l [{shapeType=Capsule, length=8, radius=4.5, shapeName=clavicle_l_capsule, center=<5 0 2>, rotation=<rot -90 0 0>}]
 clavicle_r [{shapeType=Capsule, length=8, radius=4.5, shapeName=clavicle_r_capsule, center=<-5 0 -2>, rotation=<rot -90 0 0>}]
 dyn_C_hair_back_1 [{shapeType=Capsule, length=2, radius=3, center=<3 0 3.0218>, rotation=<rot 0 0 90>}]
 dyn_C_hair_back_2 [{shapeType=Capsule, length=2, radius=3, center=<3 2.0164 1.8363>, rotation=<rot 0 0 90>}]
 dyn_C_hair_back_3 [{shapeType=Capsule, length=2, radius=3, center=<3.5 0 1.2784>, rotation=<rot 0 0 90>}]
 dyn_L_hair_back_1 [{shapeType=Capsule, length=2, radius=3, center=<3 0 3.0218>, rotation=<rot 0 0 90>}]
 dyn_L_hair_back_2 [{shapeType=Capsule, length=2, radius=3, center=<3 0 3.0218>, rotation=<rot 0 0 90>}]
 dyn_L_hair_back_3 [{shapeType=Capsule, length=2, radius=3, center=<4.1003 0 2.696>, rotation=<rot 0 0 90>}]
 dyn_L_hair_front_1 [{shapeType=Capsule, length=4, radius=2, shapeName=dyn_L_hair_front_1_capsule, center=<3 2 0>, rotation=<rot -90 0 0>}]
 dyn_L_hair_front_2 [{shapeType=Capsule, length=2, radius=2.5, shapeName=dyn_L_hair_front_2_capsule, center=<3 2 0>, rotation=<rot -90 0 0>}]
 dyn_L_hair_front_3 [{shapeType=Capsule, length=2, radius=2.5, shapeName=dyn_L_hair_front_3_capsule, center=<3 2 0>, rotation=<rot -90 0 0>}]
 dyn_L_hair_front_4 [{shapeType=Capsule, length=2, radius=3, shapeName=dyn_L_hair_front_4_capsule, center=<3 2 -0.6988>, rotation=<rot -90 0 0>}]
 dyn_R_hair_back_1 [{shapeType=Capsule, length=2, radius=2, center=<1.5906 0 1.5>, rotation=<rot 0 0 -90>}]
 dyn_R_hair_back_2 [{shapeType=Capsule, length=2, radius=2, center=<0 0 5>, rotation=<rot 0 0 90>}]
 dyn_R_hair_front_1 [{shapeType=Capsule, length=4, radius=2.25, shapeName=dyn_R_hair_front_1_capsule, center=<3 -0.4509 0>, rotation=<rot 0 0 90>}]
 dyn_R_hair_front_2 [{shapeType=Capsule, length=3, radius=2.5, shapeName=dyn_R_hair_front_2_capsule, center=<6 -0.1256 0.1199>, rotation=<rot 0 0 85>}]
 dyn_R_hair_front_3 [{shapeType=Capsule, length=2, radius=2.5, shapeName=dyn_R_hair_front_3_capsule, center=<6 -0.8123 0.15>, rotation=<rot -90 0 0>}]
 dyn_R_hair_front_4 [{
   shapeType=Capsule
   length=2
   radius=2.5
   shapeName=dyn_R_hair_front_4_capsule
   center=<3.8613 -1.6793 1.2756>
   rotation=<rot -80 179.9992 -179.9994>
  }]
 head [{shapeType=Capsule, length=0.5, radius=3.5, shapeName=head_capsule, center=<0 4 0>, rotation=<rot 90 0 0>}]
 neck_01 [{shapeType=Capsule, length=8.7, radius=3.25, shapeName=neck_01_capsule, center=<2.9192 3.7067 0>, rotation=<rot 90 180 -174.9998>}]
 neck_02 [
  {shapeType=Capsule, length=0.5, radius=15, shapeName=neck_02_capsule, center=<-11.4509 10.8315 0>, rotation=<rot 0 -90 0>}
  {shapeType=Capsule, length=0.5, radius=15, shapeName=neck_02_capsule, center=<2.3932 12.5052 0>, rotation=<rot 0 -90 0>}
 ]
 spine_05 [{shapeType=Capsule, length=10, radius=8, center=<3 0 0>}]
 upperarm_l [{shapeType=Capsule, length=35, radius=4, shapeName=upperarm_l_capsule, center=<19 0 0>, rotation=<rot -90 0 0>}]
 upperarm_r [
  {shapeType=Capsule, length=35, radius=4.5, shapeName=upperarm_r_capsule, center=<-19 0 0>, rotation=<rot -90 0 0>}
  {
   shapeType=Box
   extentX=35.5
   extentY=5.5
   extentZ=26.5
   shapeName=upperarm_r_box
   center=<-4.6634 -4.809 -12.3585>
   rotation=<rot 9.1903 15.8143 27.6174>
  }
 ]
}
constraints [
 {
  bone1Name=head
  bone2Name=neck_02
  swing1LimitDegrees=45
  swing1Motion=Free
  swing2LimitDegrees=45
  swing2Motion=Free
  twistLimitDegrees=45
  twistMotion=Free
 }
 {bone1Name=dyn_L_hair_front_1, bone2Name=base_L_hair_front, swing1LimitDegrees=22.5, swing2LimitDegrees=25, twistLimitDegrees=1}
 {bone1Name=dyn_L_hair_front_2, bone2Name=dyn_L_hair_front_1, swing1LimitDegrees=22.5, swing2LimitDegrees=22.5, twistLimitDegrees=1}
 {bone1Name=dyn_L_hair_front_3, bone2Name=dyn_L_hair_front_2, swing1LimitDegrees=22.5, swing2LimitDegrees=22.5, twistLimitDegrees=1}
 {bone1Name=dyn_L_hair_front_4, bone2Name=dyn_L_hair_front_3, swing1LimitDegrees=22.5, swing2LimitDegrees=22.5, twistLimitDegrees=1}
 {bone1Name=dyn_R_hair_front_1, bone2Name=base_R_hair_front, swing1LimitDegrees=22.5, swing2LimitDegrees=25, twistLimitDegrees=5}
 {bone1Name=dyn_R_hair_front_2, bone2Name=dyn_R_hair_front_1, swing1LimitDegrees=22.5, swing2LimitDegrees=22.5, twistLimitDegrees=5}
 {bone1Name=dyn_R_hair_front_3, bone2Name=dyn_R_hair_front_2, swing1LimitDegrees=22.5, swing2LimitDegrees=22.5, twistLimitDegrees=5}
 {bone1Name=dyn_R_hair_front_4, bone2Name=dyn_R_hair_front_3, swing1LimitDegrees=22.5, swing2LimitDegrees=22.5, twistLimitDegrees=5}
 {
  bone1Name=base_L_hair_front
  bone2Name=head
  swing1LimitDegrees=45
  swing1Motion=Free
  swing2LimitDegrees=45
  swing2Motion=Free
  twistLimitDegrees=45
  twistMotion=Free
 }
 {
  bone1Name=base_R_hair_front
  bone2Name=head
  swing1LimitDegrees=45
  swing1Motion=Free
  swing2LimitDegrees=45
  swing2Motion=Free
  twistLimitDegrees=45
  twistMotion=Free
 }
 {bone1Name=dyn_R_hair_back_1, bone2Name=base_R_hair_back, swing1LimitDegrees=0, swing2LimitDegrees=20, twistLimitDegrees=1}
 {bone1Name=dyn_R_hair_back_2, bone2Name=dyn_R_hair_back_1, swing1LimitDegrees=0, swing2LimitDegrees=20, twistLimitDegrees=1}
 {bone1Name=dyn_L_hair_back_3, bone2Name=dyn_L_hair_back_2, swing1LimitDegrees=5, swing2LimitDegrees=45, twistLimitDegrees=1}
 {bone1Name=dyn_L_hair_back_2, bone2Name=dyn_L_hair_back_1, swing1LimitDegrees=5, swing2LimitDegrees=45, twistLimitDegrees=1}
 {bone1Name=dyn_L_hair_back_1, bone2Name=base_L_hair_back, swing1LimitDegrees=5, swing2LimitDegrees=25, twistLimitDegrees=1}
 {bone1Name=dyn_C_hair_back_3, bone2Name=dyn_C_hair_back_2, swing1LimitDegrees=0, swing2LimitDegrees=20, twistLimitDegrees=1}
 {bone1Name=dyn_C_hair_back_2, bone2Name=dyn_C_hair_back_1, swing1LimitDegrees=0, swing2LimitDegrees=20, twistLimitDegrees=1}
 {bone1Name=dyn_C_hair_back_1, bone2Name=base_C_hair_back, swing1LimitDegrees=5, swing2LimitDegrees=25, twistLimitDegrees=1}
 {
  bone1Name=base_C_hair_back
  bone2Name=head
  swing1LimitDegrees=45
  swing1Motion=Free
  swing2LimitDegrees=45
  swing2Motion=Free
  twistLimitDegrees=45
  twistMotion=Free
 }
 {
  bone1Name=base_L_hair_back
  bone2Name=head
  swing1LimitDegrees=45
  swing1Motion=Free
  swing2LimitDegrees=45
  swing2Motion=Free
  twistLimitDegrees=45
  twistMotion=Free
 }
 {
  bone1Name=base_R_hair_back
  bone2Name=head
  swing1LimitDegrees=45
  swing1Motion=Free
  swing2LimitDegrees=45
  swing2Motion=Free
  twistLimitDegrees=45
  twistMotion=Free
 }
 {
  bone1Name=clavicle_l
  bone2Name=spine_05
  swing1LimitDegrees=45
  swing1Motion=Free
  swing2LimitDegrees=45
  swing2Motion=Free
  twistLimitDegrees=45
  twistMotion=Free
 }
 {
  bone1Name=upperarm_l
  bone2Name=clavicle_l
  swing1LimitDegrees=45
  swing1Motion=Free
  swing2LimitDegrees=45
  swing2Motion=Free
  twistLimitDegrees=45
  twistMotion=Free
 }
 {
  bone1Name=clavicle_r
  bone2Name=spine_05
  swing1LimitDegrees=45
  swing1Motion=Free
  swing2LimitDegrees=45
  swing2Motion=Free
  twistLimitDegrees=45
  twistMotion=Free
 }
 {
  bone1Name=upperarm_r
  bone2Name=clavicle_r
  swing1LimitDegrees=45
  swing1Motion=Free
  swing2LimitDegrees=45
  swing2Motion=Free
  twistLimitDegrees=45
  twistMotion=Free
 }
 {
  bone1Name=neck_02
  bone2Name=neck_01
  swing1LimitDegrees=45
  swing1Motion=Free
  swing2LimitDegrees=45
  swing2Motion=Free
  twistLimitDegrees=45
  twistMotion=Free
 }
 {
  bone1Name=neck_01
  bone2Name=spine_05
  swing1LimitDegrees=45
  swing1Motion=Free
  swing2LimitDegrees=45
  swing2Motion=Free
  twistLimitDegrees=45
  twistMotion=Free
 }
]
