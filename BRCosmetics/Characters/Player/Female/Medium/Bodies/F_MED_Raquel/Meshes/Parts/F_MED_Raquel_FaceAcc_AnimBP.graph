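# Animation Blueprint for Raquel's face accessory part, driving the hat and secondary motion on the back hair.
# It layers body-mesh clavicle and upper-arm bones onto the parent pose, then runs head-relative RigidBody simulations.
# Back bling, cape, targeting, crouch or front-end states select the crouching physics branch over the default, and skydiving swaps in its own simulation.
# Post-simulation constraints and bone offsets adjust the result, mostly weighted by event-graph head-tilt and state values.
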AnimBlueprint /BRCosmetics/Characters/Player/Female/Medium/Bodies/F_MED_Raquel/Meshes/Parts/F_MED_Raquel_FaceAcc_AnimBP
target_skeleton=&/F_MED_Raquel_FaceAcc_Skeleton
graphs [
 AnimGraph {nodes [
   n1 = "Transform (Modify) Bone - Bone: base_RQ_hat"(
    ComponentPose=Local To Component(LocalPose=Use cached pose 'StartPose'())
    Alpha=Get headUp() -> (headUp:double)
    Translation="0,0,2.35"
    Rotation=Get rotation values hat() -> (rotation values hat:Rotator)
    Scale="1,1,1"
   ) {
    rotationMode=BMM_Additive
    rotationSpace=BCS_BoneSpace
    scaleMode=BMM_Additive
    scaleSpace=BCS_BoneSpace
    translationMode=BMM_Additive
    translationSpace=BCS_BoneSpace
    alphaBoolBlend {blendOption=Linear}
    boneToModify {boneName=base_RQ_hat}
    translation {z=2.35}
   }
   n2 = RigidBody!(ComponentPose=@n1, Alpha, OverrideWorldGravity:Vector=Get GravityOverride(), ExternalForce) -> (Pose) {
    bClampLinearTranslationLimitToRefPose=true
    bOverrideWorldGravity=true
    overridePhysicsAsset=&/F_MED_Raquel_FaceAcc_Crouching_Physics
    simulationSpace=BaseBoneSpace
    alphaBoolBlend {blendOption=Linear}
    baseBoneRef {boneName=head}
    componentAppliedLinearAccClamp=<3000 3000 3000>
    componentLinearAccScale=<0.1 0.2 0.2>
    componentLinearVelScale=<0.1 0.2 0.2>
    overrideWorldGravity {z=-980}
    simSpaceSettings {maxAngularAcceleration=1000, maxAngularVelocity=3, maxLinearAcceleration=250, maxLinearVelocity=500, velocityScaleZ=0.1, worldAlpha=1}
   }
   n3 = Component To Local(ComponentPose=@n2)
   n4 = Use cached pose 'Skydiving_RB_Cache'()
   n5 = Get bIsSkyDiving()
   n6 = "Constraint - dyn_C_hair_back_1 (P)"(
    ComponentPose=Local To Component(LocalPose=Blend Poses by bool(BlendPose_0=@n4, BlendPose_1=@n3, bActiveValue=@n5) {blendType=Linear})
    Alpha=0.2
    ConstraintWeights_0=1
   ) {
    alphaBoolBlend {blendOption=Linear}
    boneToModify {boneName=dyn_C_hair_back_1}
    constraintSetup [{targetBone {boneName=spine_05}}]
    constraintWeights [1]
   }
   n7 = "Constraint - dyn_L_hair_back_1 (PP)"!(ComponentPose=@n6, Alpha=0.2, ConstraintWeights_0=0.2, ConstraintWeights_1:float=0.8) -> (Pose) {
    alphaBoolBlend {blendOption=Linear}
    boneToModify {boneName=dyn_L_hair_back_1}
    constraintSetup [{targetBone {boneName=spine_05}}, {targetBone {boneName=clavicle_l}}]
    constraintWeights [1,1]
   }
   n8 = "Constraint - dyn_R_hair_back_1 (PP)"!(ComponentPose=@n7, Alpha=0.2, ConstraintWeights_0=0.8, ConstraintWeights_1:float=0.2) -> (Pose) {
    alphaBoolBlend {blendOption=Linear}
    boneToModify {boneName=dyn_R_hair_back_1}
    constraintSetup [{targetBone {boneName=spine_05}}, {targetBone {boneName=clavicle_r}}]
    constraintWeights [1,1]
   }
   n9 = "Constraint - dyn_C_hair_back_2 (P)"(ComponentPose=@n8, Alpha=0.5, ConstraintWeights_0=1) {
    alphaBoolBlend {blendOption=Linear}
    boneToModify {boneName=dyn_C_hair_back_2}
    constraintSetup [{targetBone {boneName=spine_05}}]
    constraintWeights [1]
   }
   n10 = "Constraint - dyn_L_hair_back_2 (PP)"!(ComponentPose=@n9, Alpha=0.5, ConstraintWeights_0=0.2, ConstraintWeights_1:float=0.8) -> (Pose) {
    alphaBoolBlend {blendOption=Linear}
    boneToModify {boneName=dyn_L_hair_back_2}
    constraintSetup [{targetBone {boneName=spine_05}}, {targetBone {boneName=clavicle_l}}]
    constraintWeights [1,1]
   }
   n11 = "Constraint - dyn_R_hair_back_2 (PP)"!(ComponentPose=@n10, Alpha=0.7, ConstraintWeights_0=0.8, ConstraintWeights_1:float=0.2) -> (Pose) {
    alphaBoolBlend {blendOption=Linear}
    boneToModify {boneName=dyn_R_hair_back_2}
    constraintSetup [{targetBone {boneName=spine_05}}, {targetBone {boneName=clavicle_r}}]
    constraintWeights [1,1]
   }
   n12 = "Transform (Modify) Bone - Bone: dyn_R_hair_back_1"(
    ComponentPose=@n11
    Alpha=Get headUp_otherStates() -> (headUp_otherStates:double)
    Translation=Get distanceRHairStrand() -> (distanceRHairStrand:Vector)
    Rotation="0,0,10"
    Scale="1,1,1"
   ) {
    rotationMode=BMM_Additive
    scaleMode=BMM_Additive
    scaleSpace=BCS_BoneSpace
    translationMode=BMM_Additive
    translationSpace=BCS_BoneSpace
    alphaBoolBlend {blendOption=Linear}
    boneToModify {boneName=dyn_R_hair_back_1}
    rotation {roll=10}
   }
   n13 = Get headUp_Crouching() -> (headUp_Crouching:double)
   n14 = "Transform (Modify) Bone - Bone: dyn_R_hair_back_2"(ComponentPose=@n12, Alpha=@n13, Translation="0,2.5,-1", Rotation="0,0,0", Scale="1,1,1") {
    rotationMode=BMM_Additive
    scaleMode=BMM_Additive
    scaleSpace=BCS_BoneSpace
    translationMode=BMM_Additive
    translationSpace=BCS_BoneSpace
    alphaBoolBlend {blendOption=Linear}
    boneToModify {boneName=dyn_R_hair_back_2}
    translation {y=2.5, z=-1}
   }
   n15 = "Transform (Modify) Bone - Bone: dyn_L_hair_back_2"(ComponentPose=@n14, Alpha=@n13, Translation="0,0,3", Rotation="0,0,0", Scale="1,1,1") {
    rotationMode=BMM_Additive
    scaleMode=BMM_Additive
    scaleSpace=BCS_BoneSpace
    translationMode=BMM_Additive
    translationSpace=BCS_BoneSpace
    alphaBoolBlend {blendOption=Linear}
    boneToModify {boneName=dyn_L_hair_back_2}
    translation {z=3}
   }
   n16 = "Transform (Modify) Bone - Bone: dyn_L_hair_back_2"(ComponentPose=@n15, Alpha=1, Translation="0,0,1", Rotation="0,0,0", Scale="1,1,1") {
    rotationMode=BMM_Additive
    scaleMode=BMM_Additive
    scaleSpace=BCS_BoneSpace
    translationMode=BMM_Additive
    translationSpace=BCS_BoneSpace
    alphaBoolBlend {blendOption=Linear}
    boneToModify {boneName=dyn_L_hair_back_2}
    translation {z=1}
   }
   n17 = Get bIsTargeting()
   n18 = Get Is Wearing Cape() -> (Is Wearing Cape:bool)
   n19 = Get bIsCrouching()
   n20 = Get bIsInFrontEnd()
   n21 = Get isWearingBackbling()
   n22 = RigidBody!(
    ComponentPose=Local To Component(LocalPose=Use cached pose 'PreSimHairRotation_Cache'())
    Alpha
    OverrideWorldGravity=Get GravityOverride()
    ExternalForce
    ComponentLinearAccScale:Vector="0.3,0.25,0.2"
   ) -> (Pose) {
    bClampLinearTranslationLimitToRefPose=true
    bOverrideWorldGravity=true
    overridePhysicsAsset=&/F_MED_Raquel_FaceAcc_Physics
    simulationSpace=BaseBoneSpace
    alphaBoolBlend {blendOption=Linear}
    baseBoneRef {boneName=head}
    componentAppliedLinearAccClamp=<3000 3000 3000>
    componentLinearAccScale=<0.3 0.25 0.2>
    componentLinearVelScale=<0.3 0.25 0.2>
    overrideWorldGravity {z=-980}
    simSpaceSettings {maxAngularAcceleration=1000, maxAngularVelocity=3, maxLinearAcceleration=250, maxLinearVelocity=500, velocityScaleZ=0.1, worldAlpha=0.5}
   }
   n23 = Component To Local(ComponentPose=@n22)
   n24 = Use cached pose 'Skydiving_RB_Cache'()
   n25 = Get bIsSkyDiving()
   n26 = "Constraint - dyn_C_hair_back_1 (P)"(
    ComponentPose=Local To Component(LocalPose=Blend Poses by bool(BlendPose_0=@n24, BlendPose_1=@n23, bActiveValue=@n25) {blendType=Linear})
    Alpha=0.2
    ConstraintWeights_0=1
   ) {
    alphaBoolBlend {blendOption=Linear}
    boneToModify {boneName=dyn_C_hair_back_1}
    constraintSetup [{targetBone {boneName=spine_05}}]
    constraintWeights [1]
   }
   n27 = "Constraint - dyn_R_hair_back_1 (PP)"(ComponentPose=@n26, Alpha=0.2, ConstraintWeights_0=0.8, ConstraintWeights_1=0.2) {
    alphaBoolBlend {blendOption=Linear}
    boneToModify {boneName=dyn_R_hair_back_1}
    constraintSetup [{targetBone {boneName=spine_05}}, {targetBone {boneName=clavicle_r}}]
    constraintWeights [1,1]
   }
   n28 = Get headUp_PostPr() -> (headUp_PostPr:double)
   n29 = "Constraint - dyn_C_hair_back_2 (P)"(ComponentPose=@n27, Alpha=@n28, ConstraintWeights_0=1) {
    alphaBoolBlend {blendOption=Linear}
    boneToModify {boneName=dyn_C_hair_back_2}
    constraintSetup [{targetBone {boneName=spine_05}}]
    constraintWeights [1]
   }
   n30 = "Constraint - dyn_R_hair_back_2 (PP)"(ComponentPose=@n29, Alpha=@n28, ConstraintWeights_0=0.8, ConstraintWeights_1=0.2) {
    alphaBoolBlend {blendOption=Linear}
    boneToModify {boneName=dyn_R_hair_back_2}
    constraintSetup [{targetBone {boneName=spine_05}}, {targetBone {boneName=clavicle_r}}]
    constraintWeights [1,1]
   }
   n31 = "Constraint - dyn_L_hair_back_1 (PP)"(ComponentPose=@n30, Alpha=@n28, ConstraintWeights_0=0.8, ConstraintWeights_1=0.2) {
    alphaBoolBlend {blendOption=Linear}
    boneToModify {boneName=dyn_L_hair_back_1}
    constraintSetup [{targetBone {boneName=spine_05}}, {targetBone {boneName=clavicle_l}}]
    constraintWeights [1,1]
   }
   n32 = "Constraint - dyn_L_hair_back_2 (PP)"(ComponentPose=@n31, Alpha=@n28, ConstraintWeights_0=0.8, ConstraintWeights_1=0.2) {
    alphaBoolBlend {blendOption=Linear}
    boneToModify {boneName=dyn_L_hair_back_2}
    constraintSetup [{targetBone {boneName=spine_05}}, {targetBone {boneName=clavicle_l}}]
    constraintWeights [1,1]
   }
   n33 = "Transform (Modify) Bone - Bone: dyn_R_hair_back_1"(ComponentPose=@n32, Alpha=Get headUp_PostPr(), Translation="-4,0,0", Rotation="0,0,0", Scale="1,1,1") {
    rotationMode=BMM_Additive
    scaleMode=BMM_Additive
    scaleSpace=BCS_BoneSpace
    translationMode=BMM_Additive
    translationSpace=BCS_BoneSpace
    alphaBoolBlend {blendOption=Linear}
    boneToModify {boneName=dyn_R_hair_back_1}
    translation {x=-4}
   }
   n34 = "Transform (Modify) Bone - Bone: base_RQ_hat"(ComponentPose=@n33, Alpha=Get isSkydivingAlpha() -> (isSkydivingAlpha:double), Translation="0,0,1", Rotation="0,0,25", Scale="1,1,1") {
    rotationMode=BMM_Additive
    scaleMode=BMM_Additive
    scaleSpace=BCS_BoneSpace
    translationMode=BMM_Additive
    translationSpace=BCS_BoneSpace
    alphaBoolBlend {blendOption=Linear}
    alphaScaleBiasClamp {bInterpResult=true}
    boneToModify {boneName=base_RQ_hat}
    rotation {roll=25}
    translation {z=1}
   }
   Output Pose(Result=Blend Poses by bool(
     BlendPose_0=Component To Local(ComponentPose=@n16)
     BlendPose_1=Component To Local(ComponentPose=@n34)
     BlendTime_1=0.25
     bActiveValue=OR Boolean!(A=@n21, B=@n18, C:bool=@n17, D:bool=@n19, E:bool=@n20) -> (ReturnValue)
    ) {blendType=Linear}) {layerGroup=None}
   n35 = "Transform (Modify) Bone - Bone: base_RQ_hat"(
    ComponentPose=Local To Component(LocalPose=Use cached pose 'StartPose'())
    Alpha=Get headUp()
    Translation="0,0,3"
    Rotation=Get rotation values hat()
    Scale="1,1,1"
   ) {
    rotationMode=BMM_Additive
    rotationSpace=BCS_BoneSpace
    scaleMode=BMM_Additive
    scaleSpace=BCS_BoneSpace
    translationMode=BMM_Additive
    translationSpace=BCS_BoneSpace
    alphaBoolBlend {blendOption=Linear}
    boneToModify {boneName=base_RQ_hat}
    translation {z=3}
   }
   n36 = "Transform (Modify) Bone - Bone: master_RQ_main_hair_back"(ComponentPose=@n35, Alpha=Get isSkydivingAlpha(), Translation="0,2,0", Rotation="35,-5,0", Scale="1,1,1") {
    rotationMode=BMM_Additive
    rotationSpace=BCS_BoneSpace
    scaleMode=BMM_Additive
    scaleSpace=BCS_BoneSpace
    translationMode=BMM_Additive
    translationSpace=BCS_BoneSpace
    alphaBoolBlend {blendOption=Linear}
    boneToModify {boneName=master_RQ_main_hair_back}
    rotation {pitch=35, yaw=-5}
    translation {y=2}
   }
   n37 = "Transform (Modify) Bone - Bone: base_C_hair_back"(ComponentPose=@n36, Alpha=1, Translation="0,0,0", Rotation="-60,0,0", Scale="1,1,1") {
    rotationMode=BMM_Additive
    rotationSpace=BCS_BoneSpace
    scaleMode=BMM_Additive
    scaleSpace=BCS_BoneSpace
    translationMode=BMM_Additive
    translationSpace=BCS_BoneSpace
    alphaBoolBlend {blendOption=Linear}
    boneToModify {boneName=base_C_hair_back}
    rotation {pitch=-60}
   }
   n38 = "Transform (Modify) Bone - Bone: base_R_hair_back"(
    ComponentPose=@n37
    Alpha=1
    Translation=Get distanceRHairStrand_Skydiving() -> (distanceRHairStrand_Skydiving:Vector)
    Rotation=Get rotation values right hair back Skydiving() -> (rotation values right hair back Skydiving:Rotator)
    Scale="1,1,1"
   ) {
    rotationMode=BMM_Additive
    rotationSpace=BCS_BoneSpace
    scaleMode=BMM_Additive
    scaleSpace=BCS_BoneSpace
    translationMode=BMM_Additive
    translationSpace=BCS_BoneSpace
    alphaBoolBlend {blendOption=Linear}
    boneToModify {boneName=base_R_hair_back}
   }
   Save cached pose 'PreSimHairRotation_Cache'(Pose=Component To Local(ComponentPose=@n38)) {cacheName=PreSimHairRotation_Cache}
   n39 = RigidBody(
    ComponentPose=Local To Component(LocalPose=Use cached pose 'PreSimHairRotation_Cache'())
    OverrideWorldGravity=Get GravityOverride()
    ComponentLinearAccScale="0.3,0.25,0.2"
   ) {
    bClampLinearTranslationLimitToRefPose=true
    bOverrideWorldGravity=true
    overridePhysicsAsset=&/F_MED_Raquel_FaceAcc_Skydiving_Physics
    simulationSpace=BaseBoneSpace
    alphaBoolBlend {blendOption=Linear}
    baseBoneRef {boneName=head}
    componentAppliedLinearAccClamp=<3000 3000 3000>
    componentLinearAccScale=<0.3 0.25 0.2>
    componentLinearVelScale=<0.275 0.2 0.2>
    overrideWorldGravity {z=-980}
    simSpaceSettings {maxAngularAcceleration=500, maxAngularVelocity=3, maxLinearAcceleration=250, maxLinearVelocity=3000, velocityScaleZ=0.3}
   }
   Save cached pose 'Skydiving_RB_Cache'(Pose=Component To Local(ComponentPose=@n39)) {cacheName=Skydiving_RB_Cache}
   n40 = Layered blend per bone(
    BasePose=Copy Pose From Mesh() {bCopyCurves=true, bUseAttachedParent=true}
    BlendPoses_0=Copy Pose From Mesh(SourceMeshComponent=Get Body_Mesh())
   ) {layerSetup [{branchFilters [{boneName=clavicle_l}, {boneName=clavicle_r}, {boneName=upperarm_r}, {boneName=upperarm_l}]}]}
   Save cached pose 'StartPose'(Pose=@n40) {cacheName=StartPose}
  ]}
 EventGraph {nodes [
   Event OnCharacterCustomizationLoaded()
   > n1 = Cast To PlayerPawn_Athena(Object=TryGetPawnOwner())
   > Set Body_Mesh(Body_Mesh=GetSkeletalMeshForPartType(self=@n1))
   Event BlueprintInitializeAnimation()
   > Set Skel Mesh(Skel Mesh=GetOwnerSkeletalMeshComponent())
   Event BlueprintUpdateAnimation()
   > n2 = Sequence!(execute) -> (then_0, then_1, then_2:Exec, then_3:Exec)
   n3 = CheckForBackbling()
   Set isWearingBackbling(
    execute=@n2.then_2
    isWearingBackbling=AND Boolean!(A=@n3, B=NOT Boolean(A=@n3.bIsBackblingACape), C:bool=NOT Boolean(A=@n3.bIsBackblingLong)) -> (ReturnValue)
   )
   > Set Is Wearing Cape(Is Wearing Cape=OR Boolean!(A=@n3.bIsBackblingACape, B=@n3.bIsBackblingLong) -> (ReturnValue))
   n4 = Get Is Wearing Cape()
   n5 = Get isWearingBackbling()
   n6 = Get bIsSkydiveFloating()
   n7 = Get bIsSkydiveDiveMode()
   > n8 = Branch(Condition=AND Boolean!(A=OR Boolean(A=@n5, B=@n4), B=OR Boolean(A=@n6, B=@n7)) -> (ReturnValue))
   > Set distanceRHairStrand_Skydiving(distanceRHairStrand_Skydiving:Vector="-5,0,0") -> (Output_Get:Vector)
   > Set rotation values right hair back Skydiving(rotation values right hair back Skydiving:Rotator="0, 0, 0") -> (Output_Get:Rotator)
   Set distanceRHairStrand_Skydiving(execute=@n8.else, distanceRHairStrand_Skydiving="0, 0, 0")
   > Set rotation values right hair back Skydiving(rotation values right hair back Skydiving="0,0,55")
   n9 = Branch(execute=@n2.then_1, Condition=Get bIsSkyDiving())
   > Set headUp(headUp:double=0) -> (Output_Get:double)
   > Set headUp_Crouching(headUp_Crouching:double=0) -> (Output_Get:double)
   > Set headUp_PostPr(headUp_PostPr:double=0) -> (Output_Get:double)
   > Set headUp_otherStates(headUp_otherStates:double=0) -> (Output_Get:double)
   n10 = Get Skel Mesh()
   n11 = Set headUp(
    execute=@n9.else
    headUp=GetAlphaFromReferenceTransformDelta(SkeletalMeshComponent=@n10, BoneName=spine_05, BaseName=head, InputRange=(X=0,Y=150), TwistAxis="0,0,1")
   )
   > Set headUp_otherStates(headUp_otherStates="Clamp (Float)"(
     Value=GetAlphaFromReferenceTransformDelta(SkeletalMeshComponent=@n10, BoneName=spine_05, BaseName=head, InputRange=(X=0,Y=65), TwistAxis="0,0,1")
     Min=0.5
     Max=0.75
    ))
   > Set headUp_Crouching(headUp_Crouching=float * float(A=@n11, B:bool=Get bIsCrouching()) {is=title K2Node_PromotableOperator})
   > Set headUp_PostPr(headUp_PostPr=float * float(A=Get headUp(), B:double=2))
   > n12 = Branch(Condition=Get bIsCrouching())
   n13 = Get headUp_Crouching()
   > Set rotation values hat(rotation values hat:Rotator=MakeRotator(Roll=float * float(A=@n13, B=10), Pitch=float * float(A=@n13, B=90))) -> (Output_Get:Rotator)
   > Set distanceRHairStrand(distanceRHairStrand:Vector="-4,1,0") -> (Output_Get:Vector)
   n14 = Branch(execute=@n12.else, Condition=Get bIsSkyDiving())
   > Set rotation values hat(rotation values hat="100,0,0")
   > Set distanceRHairStrand(distanceRHairStrand="-6,0,0")
   Set rotation values hat(execute=@n14.else, rotation values hat="0, 0, 0")
   > Set distanceRHairStrand(distanceRHairStrand="-4,2,0")
   n15 = Gravity Override State Function DEPRECATED(
    execute=@n2.then_0
    CharacterPartAnimInstance=Self-Reference()
    GameTimeInSeconds=GetGameTimeInSeconds()
    GravityOverrideParams=Get Gravity Override Params()
   )
   > Set GravityOverride(GravityOverride=@n15)
   > n16 = Branch(Condition=Get bIsSkyDiving())
   > n17 = Branch(Condition=Get bIsParachuteOpened())
   > Set isSkydivingAlpha(isSkydivingAlpha:double=0.25) -> (Output_Get:double)
   Set isSkydivingAlpha(execute=@n17.else, isSkydivingAlpha=1)
   Set isSkydivingAlpha(execute=@n16.else, isSkydivingAlpha=0)
  ]}
]
variables [
 Body_Mesh {type=SkeletalMeshComponent}
 distanceRHairStrand {type=Vector}
 distanceRHairStrand_Skydiving {type=Vector}
 Gravity Override Params {type=GravityOverrideParamsStruct}
 GravityOverride {type=Vector}
 headUp {type=number}
 headUp_Crouching {type=number}
 headUp_otherStates {type=number}
 headUp_PostPr {type=number}
 Is Wearing Cape {type=boolean}
 isSkydivingAlpha {type=number}
 isWearingBackbling {type=boolean}
 rotation values hat {type=Rotator}
 rotation values right hair back Skydiving {type=Rotator}
 Skel Mesh {type=SkeletalMeshComponent}
]
